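# Post-copy AnimBlueprint for the male medium Golf body cosmetic, correcting mesh deformation on a pose copied from the attached parent mesh.
# Additive bone offsets and a chest-to-spine_05 rotation constraint feed sculpter and twist Control Rigs, followed by knee, back, neck and clavicle fixes.
# The event graph combines crouch, jump and fall states into two flags that blend in left knee-fold and both-knee scale corrections.
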AnimBlueprint /BRCosmetics/Characters/Player/Male/Medium/Bodies/M_MED_Golf/Meshes/M_MED_Golf_AnimBP
target_skeleton=&/M_MED_Golf_Skeleton
graphs [
 AnimGraph {nodes [
   "Constraint - dyn_ChestFix (R)"(Alpha=1, ConstraintWeights_0=1) {
    alphaBoolBlend {blendOption=Linear}
    boneToModify {boneName=dyn_ChestFix}
    constraintSetup [{transformType=Rotation, targetBone {boneName=spine_05}}]
    constraintWeights [1]
   }
   n1 = "Transform (Modify) Bone - Bone: dyn_l_arm_bulge"!(
    ComponentPose=Local To Component(LocalPose=Copy Pose From Mesh() {bUseAttachedParent=true})
    bAlphaBoolEnabled:bool=True
    Translation="0,0,-3"
    Rotation="0,0,0"
    Scale="1,1,1"
   ) -> (Pose) {
    alphaInputType=Bool
    rotationMode=BMM_Additive
    rotationSpace=BCS_BoneSpace
    scaleMode=BMM_Additive
    scaleSpace=BCS_BoneSpace
    translationMode=BMM_Additive
    translationSpace=BCS_BoneSpace
    alphaBoolBlend {blendInTime=0.1, blendOption=Linear, blendOutTime=0.1}
    boneToModify {boneName=dyn_l_arm_bulge}
    translation {z=-3}
   }
   n2 = "Transform (Modify) Bone - Bone: dyn_r_arm_bulge"!(ComponentPose=@n1, bAlphaBoolEnabled:bool=True, Translation="0,0,3", Rotation="0,0,0", Scale="1,1,1") -> (Pose) {
    alphaInputType=Bool
    rotationMode=BMM_Additive
    rotationSpace=BCS_BoneSpace
    scaleMode=BMM_Additive
    scaleSpace=BCS_BoneSpace
    translationMode=BMM_Additive
    translationSpace=BCS_BoneSpace
    alphaBoolBlend {blendInTime=0.1, blendOption=Linear, blendOutTime=0.1}
    boneToModify {boneName=dyn_r_arm_bulge}
    translation {z=3}
   }
   n3 = "Transform (Modify) Bone - Bone: dyn_insideCollar_Fix"!(ComponentPose=@n2, bAlphaBoolEnabled:bool=True, Translation="-3,0,0", Rotation="0,0,0", Scale="1,1.5,1") -> (Pose) {
    alphaInputType=Bool
    rotationMode=BMM_Additive
    scaleMode=BMM_Additive
    translationMode=BMM_Additive
    alphaBoolBlend {blendInTime=0.1, blendOption=Linear, blendOutTime=0.1}
    boneToModify {boneName=dyn_insideCollar_Fix}
    scale {y=1.5}
    translation {x=-3}
   }
   n4 = "Constraint - dyn_ChestFix (R)"(ComponentPose=@n3, Alpha=1, ConstraintWeights_0=1) {
    alphaBoolBlend {blendOption=Linear}
    boneToModify {boneName=dyn_ChestFix}
    constraintSetup [{transformType=Rotation, targetBone {boneName=spine_05}}]
    constraintWeights [1]
   }
   n5 = "Transform (Modify) Bone - Bone: dyn_ChestFix"!(ComponentPose=@n4, bAlphaBoolEnabled:bool=True, Translation="0,0,0", Rotation="0,0,0", Scale="1,1.5,1") -> (Pose) {
    alphaInputType=Bool
    rotationMode=BMM_Additive
    scaleMode=BMM_Additive
    translationMode=BMM_Additive
    alphaBoolBlend {blendInTime=0.1, blendOption=Linear, blendOutTime=0.1}
    boneToModify {boneName=dyn_ChestFix}
    scale {y=1.5}
   }
   n6 = Control Rig(Source=Component To Local(ComponentPose=@n5)) {
    alphaBoolBlend {blendOption=Linear}
    controlRigAssetReference {blueprintRigClass=&/M_MED_Golf_Deform_CtrlRig_sculpters.M_MED_Golf_Deform_CtrlRig_sculpters_C}
   }
   n7 = Control Rig(Source=@n6) {
    alphaBoolBlend {blendOption=Linear}
    controlRigAssetReference {blueprintRigClass=/DefaultCosmeticsAthena/Characters/Meshes/M_MED_Golf_Deform_CtrlRig_twists.M_MED_Golf_Deform_CtrlRig_twists_C}
   }
   n8 = Get Crouch_Jump_Fall() -> (Crouch_Jump_Fall:bool)
   n9 = "Transform (Modify) Bone - Bone: dyn_l_kneeFold"!(ComponentPose=Local To Component(LocalPose=@n7), bAlphaBoolEnabled:bool=@n8, Translation="0,0,0", Rotation="0,0,-30", Scale="1,1,0.55") -> (Pose) {
    alphaInputType=Bool
    rotationMode=BMM_Additive
    rotationSpace=BCS_BoneSpace
    scaleMode=BMM_Additive
    scaleSpace=BCS_BoneSpace
    translationMode=BMM_Additive
    translationSpace=BCS_BoneSpace
    alphaBoolBlend {blendInTime=0.1, blendOption=Linear, blendOutTime=0.1}
    boneToModify {boneName=dyn_l_kneeFold}
    rotation {roll=-30}
    scale {z=0.55}
   }
   n10 = "Transform (Modify) Bone - Bone: dyn_l_kneeFold"(ComponentPose=@n9, bAlphaBoolEnabled=@n8, Translation="0,0,0", Rotation="0,0,0", Scale="1,1,0.55") {
    alphaInputType=Bool
    rotationMode=BMM_Additive
    rotationSpace=BCS_BoneSpace
    scaleMode=BMM_Additive
    scaleSpace=BCS_BoneSpace
    translationMode=BMM_Additive
    translationSpace=BCS_BoneSpace
    alphaBoolBlend {blendInTime=0.1, blendOption=Linear, blendOutTime=0.1}
    boneToModify {boneName=dyn_l_kneeFold}
    scale {z=0.55}
   }
   n11 = Get Jumping_Fall() -> (Jumping_Fall:bool)
   n12 = "Transform (Modify) Bone - Bone: deform_knee_fr_01_l"!(ComponentPose=@n10, bAlphaBoolEnabled:bool=@n11, Translation="0,0,0", Rotation="0,0,0", Scale="1.2,1.2,1.2") -> (Pose) {
    alphaInputType=Bool
    rotationMode=BMM_Additive
    rotationSpace=BCS_BoneSpace
    scaleMode=BMM_Additive
    scaleSpace=BCS_BoneSpace
    translationMode=BMM_Additive
    translationSpace=BCS_BoneSpace
    alphaBoolBlend {blendInTime=0.1, blendOption=Linear, blendOutTime=0.1}
    boneToModify {boneName=deform_knee_fr_01_l}
    scale=<1.2 1.2 1.2>
   }
   n13 = "Transform (Modify) Bone - Bone: deform_knee_fr_01_r"!(ComponentPose=@n12, bAlphaBoolEnabled:bool=@n11, Translation="0,0,0", Rotation="0,0,0", Scale="1.2,1.2,1.2") -> (Pose) {
    alphaInputType=Bool
    rotationMode=BMM_Additive
    rotationSpace=BCS_BoneSpace
    scaleMode=BMM_Additive
    scaleSpace=BCS_BoneSpace
    translationMode=BMM_Additive
    translationSpace=BCS_BoneSpace
    alphaBoolBlend {blendInTime=0.1, blendOption=Linear, blendOutTime=0.1}
    boneToModify {boneName=deform_knee_fr_01_r}
    scale=<1.2 1.2 1.2>
   }
   n14 = "Transform (Modify) Bone - Bone: dyn_backFix"!(ComponentPose=@n13, bAlphaBoolEnabled:bool=True, Translation="0,-12,0", Rotation="0,0,0", Scale="1,1,1") -> (Pose) {
    alphaInputType=Bool
    rotationMode=BMM_Additive
    rotationSpace=BCS_BoneSpace
    scaleMode=BMM_Additive
    scaleSpace=BCS_BoneSpace
    translationMode=BMM_Additive
    translationSpace=BCS_BoneSpace
    alphaBoolBlend {blendInTime=0.1, blendOption=Linear, blendOutTime=0.1}
    boneToModify {boneName=dyn_backFix}
    translation {y=-12}
   }
   n15 = "Transform (Modify) Bone - Bone: dyn_neck"!(ComponentPose=@n14, bAlphaBoolEnabled:bool=True, Translation="0,0,0", Rotation="0,0,0", Scale="1.8,1.8,1.8") -> (Pose) {
    alphaInputType=Bool
    rotationMode=BMM_Additive
    rotationSpace=BCS_BoneSpace
    scaleMode=BMM_Additive
    scaleSpace=BCS_BoneSpace
    translationMode=BMM_Additive
    translationSpace=BCS_BoneSpace
    alphaBoolBlend {blendInTime=0.1, blendOption=Linear, blendOutTime=0.1}
    boneToModify {boneName=dyn_neck}
    scale=<1.8 1.8 1.8>
   }
   n16 = "Transform (Modify) Bone - Bone: dyn_l_clavCorrect"!(ComponentPose=@n15, bAlphaBoolEnabled:bool=True, Translation="4,0,3", Rotation="0,0,0", Scale="1,1,1") -> (Pose) {
    alphaInputType=Bool
    rotationMode=BMM_Additive
    rotationSpace=BCS_BoneSpace
    scaleMode=BMM_Additive
    scaleSpace=BCS_BoneSpace
    translationMode=BMM_Additive
    translationSpace=BCS_BoneSpace
    alphaBoolBlend {blendInTime=0.1, blendOption=Linear, blendOutTime=0.1}
    boneToModify {boneName=dyn_l_clavCorrect}
    translation {x=4, z=3}
   }
   n17 = "Transform (Modify) Bone - Bone: dyn_r_clavCorrect"!(ComponentPose=@n16, bAlphaBoolEnabled:bool=True, Translation="-4,0,-3", Rotation="0,0,0", Scale="1.8,1.8,1.8") -> (Pose) {
    alphaInputType=Bool
    rotationMode=BMM_Additive
    rotationSpace=BCS_BoneSpace
    scaleMode=BMM_Additive
    scaleSpace=BCS_BoneSpace
    translationMode=BMM_Additive
    translationSpace=BCS_BoneSpace
    alphaBoolBlend {blendInTime=0.1, blendOption=Linear, blendOutTime=0.1}
    boneToModify {boneName=dyn_r_clavCorrect}
    scale=<1.8 1.8 1.8>
    translation {x=-4, z=-3}
   }
   Output Pose(Result=Component To Local(ComponentPose=@n17)) {layerGroup=None}
  ]}
 EventGraph {nodes [
   Event BlueprintUpdateAnimation()
   > n1 = Sequence()
   > Set Crouch_Jump_Fall(Crouch_Jump_Fall:bool=OR Boolean!(A=Get bIsCrouching(), B=Get bIsJumping(), C:bool=Get bIsFalling()) -> (ReturnValue)) -> (Output_Get:bool)
   Set Jumping_Fall(execute=@n1.then_1, Jumping_Fall:bool=OR Boolean(A=false, B=Get bIsJumping(), C=Get bIsFalling())) -> (Output_Get:bool)
  ]}
]
variables [
 Crouch_Jump_Fall {type=boolean}
 Jumping_Fall {type=boolean}
]
